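# Animation Blueprint for the IronLog shotgun weapon skeleton, posing the gun and playing slot animations over it.
# The output pose is only the two-bullet pose passed through DefaultSlot.
# A separate, unconnected blend tree selects bullet-count poses from AmmoRemaining and FireRightComplete, using DefaultSlot and FireRight slots.
# Notifies set FireRightComplete when right fire completes and clear it when reload completes; the update event is empty.
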AnimBlueprint /WildEstateWeapons/Gameplay/IronLog/Animation/IronLog_Shotgun_AnimBP
target_skeleton=/WildEstateWeapons/Assets/IronLog_Shotgun/IronLog_Shotgun_Skeleton
graphs [
 AnimGraph {nodes [
   n1 = Sequence Player 'IronLog_Shotgun_Pose_WEP_2Bullets'() {bLoopAnimation=false, sequence=&/IronLog_Shotgun_Pose_WEP_2Bullets}
   n2 = Sequence Player 'IronLog_Shotgun_Pose_WEP_1Bullet'() {sequence=&/IronLog_Shotgun_Pose_WEP_1Bullet}
   n3 = "integer >= integer"(A=Get AmmoRemaining() -> (AmmoRemaining:int=0), B=2)
   n4 = Sequence Player 'IronLog_Shotgun_Pose_WEP_1Bullet'() {sequence=&IronLog_Shotgun_Pose_WEP_1Bullet}
   n5 = Sequence Player 'IronLog_Shotgun_Pose_WEP_0Bullets'() {bLoopAnimation=false, sequence=&/IronLog_Shotgun_Pose_WEP_0Bullets}
   n6 = "integer <= integer"(A=Get AmmoRemaining() -> (AmmoRemaining=0))
   Blend Poses by bool(
    BlendPose_0=Slot 'DefaultSlot'(Source=Blend Poses by bool(BlendPose_0=@n5, BlendPose_1=@n4, BlendTime_0=0, BlendTime_1=0, bActiveValue=@n6))
    BlendPose_1=Slot 'FireRight'(Source=Blend Poses by bool(BlendPose_0=@n1, BlendPose_1=@n2, BlendTime_0=0, BlendTime_1=0, bActiveValue=@n3)) {slotName=FireRight}
    BlendTime_0=0
    BlendTime_1=0
    bActiveValue=Get FireRightComplete() -> (FireRightComplete:bool)
   )
   Output Pose(Result=Slot 'DefaultSlot'(Source=Sequence Player 'IronLog_Shotgun_Pose_WEP_2Bullets'() {bLoopAnimation=false, sequence=&IronLog_Shotgun_Pose_WEP_2Bullets}))
  ]}
 EventGraph {nodes [
   AnimNotify_OnFireRightComplete() {is=title K2Node_Event}
   > Set FireRightComplete(FireRightComplete:bool=true) -> (Output_Get:bool)
   AnimNotify_OnReloadComplete() {is=title K2Node_Event}
   > Set FireRightComplete(FireRightComplete=false)
   TryGetPawnOwner()
   Event BlueprintUpdateAnimation()
  ]}
]
variables [FireRightComplete {type=boolean}]
