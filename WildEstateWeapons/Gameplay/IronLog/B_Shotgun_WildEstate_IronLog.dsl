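; Blueprint for the IronLog Wild Estate shotgun, handling its cosmetic paint impact decals and a Niagara tank-fill effect.
; Each pellet impact spawns a colored decal and expands min/max bounds in a local space built from the first hit.
; At the tenth impact, a utility check can replace the pellet decals with one scaled, oriented decal.
; Firing drains the tank and clears impact data, while reload start refills it.
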(event FortWeapon|EventOnPlayImpactFX
  (params (HitResult :type Hit%20Result%20Structure%20%28by%20ref%29) (ImpactPhysicalSurface :type EPhysicalSurface%20Enum :default "SurfaceType_Default") (SpawnedPSC :type FXSystem%20Component%20Object%20Reference))
  (bind (_bblockinghit _binitialoverlap _time _distance _location _impactpoint _normal _impactnormal _physmat _hitactor _hitcomponent _hitbonename _bonename _hititem _elementindex _faceindex _tracestart _traceend) (Collision|BreakHitResult HitResult))
  (bind _weapondirection (Variables|Default|GetWeaponDirection))
  (bind _impactnormal_2 (Variables|EffectsDecal|GetImpactNormal))
  (bind _minpos (Variables|EffectsDecal|GetMinPos))
  (bind _maxpos (Variables|EffectsDecal|GetMaxPos))
  (bind _decaltransform (Variables|EffectsDecal|GetDecalTransform))
  (bind (_centerpos _impactnormal_3 _weapondirection_2 _scale _rotation _single) (Utilities|GetCenterandScaleOfImpacPoints _decaltransform _maxpos _minpos _impactnormal_2 _weapondirection))
  (|CallParentFunction HitResult ImpactPhysicalSurface SpawnedPSC)
  (Effects|Components|ParticleSystem|SetFloatParameter SpawnedPSC "DebrisAmount" 0.1)
  (bind _output_get (Variables|EffectsDecal|SetImpact%20Point _impactpoint))
  (Utilities|Array|Add (Variables|EffectsDecal|GetImpactPos) _output_get)
  (bind _output_get_2 (Variables|EffectsDecal|SetImpact%20Normal _impactnormal))
  (Utilities|Array|Add (Variables|EffectsDecal|GetImpactNormal) _output_get_2)
  (bind _decal (EffectsDecal|SpawnPelletDecals (Variables|EffectsDecal|GetImpact%20Point) (Variables|EffectsDecal|GetImpact%20Normal)))
  (Utilities|Array|Add (Variables|EffectsDecal|GetPelletDecals) _decal)
  (if (== (Utilities|Array|Length (Variables|EffectsDecal|GetImpactPos)) 1)
    (bind _returnvalue (CallFunction|CreateLocalSpceonFirstImpact (Variables|EffectsDecal|GetImpact%20Point) (Variables|EffectsDecal|GetImpact%20Normal) _tracestart _traceend))
    (bind _output_get_3 (Variables|EffectsDecal|SetDecalTransform _returnvalue))
    (Variables|EffectsDecal|SetMinPos (Math|Transform|InverseTransformLocation _output_get_3 (Variables|EffectsDecal|GetImpact%20Point)))
    (Variables|EffectsDecal|SetMaxPos (Math|Transform|InverseTransformLocation _output_get_3 (Variables|EffectsDecal|GetImpact%20Point)))
    (Variables|Default|SetWeaponDirection (Math|Vector|GetUnitDirection(Vector) _tracestart _traceend))
    (else
      (Variables|EffectsDecal|SetMinPos (Math|Vector|VectorComponentMin (Variables|EffectsDecal|GetMinPos) (Math|Transform|InverseTransformLocation (Variables|EffectsDecal|GetDecalTransform) (Variables|EffectsDecal|GetImpact%20Point))))
      (Variables|EffectsDecal|SetMaxPos (Math|Vector|VectorComponentMax (Variables|EffectsDecal|GetMaxPos) (Math|Transform|InverseTransformLocation (Variables|EffectsDecal|GetDecalTransform) (Variables|EffectsDecal|GetImpact%20Point))))
      (if (== 10 (Utilities|Array|Length (Variables|EffectsDecal|GetImpactPos)))
        (if _single
          (CallFunction|SpawnSingleDecal _centerpos _impactnormal_3 _weapondirection_2 (Math|Vector|MakeVector (Math|Float|Clamp(Float) (.x _scale) 1e+01 1.5e+02) (.y _scale) (.z _scale)) _rotation)
          (for _array_element (Variables|EffectsDecal|GetPelletDecals)
            (Components|DestroyComponent _array_element)))))))

(event EventOnPlayWeaponFireFX
  (params (bPersistentFire :type Boolean) (bSecondaryFire :type Boolean))
  (|CallParentFunction bPersistentFire bSecondaryFire)
  (Utilities|FlowControl|Sequence
    (:then_0
      (bind _output_get (Variables|Default|SetTank%20Percentage (- (Variables|Default|GetTank%20Percentage) 0.4)))
      (Niagara|SetNiagaraVariable(Float) (Variables|Default|GetTank%20FX) "Fill_Percentage" _output_get)
      (if (<= _output_get 0.1)
        (Rendering|SetVisibility (Variables|Default|GetTank%20FX) false false)))
    (:then_1
      (Utilities|Array|Clear (Variables|EffectsDecal|GetImpactPos))
      (Utilities|Array|Clear (Variables|EffectsDecal|GetImpactNormal))
      (Variables|EffectsDecal|SetMinPos "0, 0, 0")
      (Variables|EffectsDecal|SetMaxPos "0, 0, 0")
      (Utilities|Array|Clear (Variables|EffectsDecal|GetPelletDecals)))))

(event EventOnPlayReloadFX
  (params (ReloadStage :type EFortReloadFXState%20Enum :default "ReloadStart"))
  (|CallParentFunction ReloadStage)
  (switch Utilities|FlowControl|Switch|SwitchonEFortReloadFXState ReloadStage
    (:ReloadStart
      (bind _output_get (Variables|Default|SetTank%20Percentage 0.8))
      (Niagara|SetNiagaraVariable(Float) (Variables|Default|GetTank%20FX) "Fill_Percentage" _output_get)
      (Rendering|SetVisibility (Variables|Default|GetTank%20FX) true false))))

(event FortWeapon|EventOnEquippedAndAttached
  (|CallParentFunction)
  (Networking|CanExecuteCosmeticEvents
    (:True
      (if (> 2 (FortWeapon|GetLocalMagazineAmmoCount))
        (Variables|Default|SetTank%20Percentage 0.4)
        (Niagara|SetNiagaraVariable(Float) (Variables|Default|GetTank%20FX) "Fill_Percentage" (Variables|Default|GetTank%20Percentage))))))

(fn CreateLocalSpceOnFirstImpact
  (params (ImpactPoint :type Vector) (ImpactNormal :type Vector) (TraceStart :type Vector) (TraceEnd :type Vector))
  ; Set LocalSpace to first Point in array
  (return (Math|Transform|MakeTransform ImpactPoint (Math|Rotator|MakeRotfromXY (* ImpactNormal -1.0) (* (Math|Vector|GetUnitDirection(Vector) TraceStart TraceEnd) -1.0)))))

(fn GetRandomDecalMaterial
  (return (Variables|Default|GetPaintDecalMaterial)))

(fn SpawnPelletDecals
  (params (ImpactLocation :type Vector) (ImpactNormal :type Vector))
  (bind _pellet_decal_material (Variables|EffectsDecal|GetPellet%20Decal%20Material))
  (if (ScalableFloat|GetValueasBool (Variables|Default|GetHF_EnableEnvironmentDecal))
    (Utilities|IsValid _pellet_decal_material
      (:Is%20Valid
        (bind _returnvalue (Rendering|Decal|SpawnDecalatLocation _pellet_decal_material (Math|Vector|MakeVector 1e+01 (Math|Random|RandomFloatinRange (Variables|Default|GetPaintDecalSizeMin) (Variables|Default|GetPaintDecalSizeMax)) (Math|Random|RandomFloatinRange (Variables|Default|GetPaintDecalSizeMin) (Variables|Default|GetPaintDecalSizeMax))) ImpactLocation (Math|Conversions|RotationFromXVector (* ImpactNormal -1.0)) (Variables|Default|GetPaintDecalLifetime)))
        (Transformation|AddLocalRotation _returnvalue (Math|Rotator|MakeRotator (Math|Random|RandomFloatinRange 0.0 3.6e+02)) false false)
        (Rendering|Components|Decal|SetFadeIn _returnvalue 0.0 (Variables|Default|GetPaintDecalFadeInDuration))
        (Rendering|Components|Decal|SetFadeOut _returnvalue (Variables|Default|GetPaintDecalDelayBeforeFade) (Variables|Default|GetPaintDecalFadeOutDuration) true)
        (Rendering|Components|Decal|SetDecalColor _returnvalue (Utilities|Array|Get(aref) (Variables|EffectsDecal|GetDecalPelletColor) (Utilities|Array|Length (Variables|EffectsDecal|GetImpactPos))))
        (return _returnvalue)))))

(fn SpawnSingleDecal
  (params (SpawnLocation :type Vector) (HitNormal :type Vector) (WeaponDirection :type Vector) (Size :type Vector) (Rotation :type Rotator))
  (bind _returnvalue (Math|Rotator|GetRotationXVector Rotation))
  (bind _y (.y _returnvalue))
  (bind _x (.x _returnvalue))
  (bind _returnvalue_2 (Math|Vector|MakeVector _x _y))
  (bind _returnvalue_3 (Game|GetPlayerPawn 0))
  (bind _returnvalue_4 (Transformation|GetActorForwardVector _returnvalue_3))
  (bind _y_2 (.y _returnvalue_4))
  (bind _x_2 (.x _returnvalue_4))
  (bind _returnvalue_5 (Math|Vector|MakeVector _x_2 _y_2))
  (bind _returnvalue_6 (Utilities|Operators|Equal(==) _returnvalue_5 _returnvalue_2 0.3))
  (bind _returnvalue_7 (Utilities|Operators|Equal(==) _returnvalue "0.000000,0.000000,-1.000000" 0.2))
  (bind _returnvalue_8 (Utilities|Operators|Equal(==) _returnvalue "0.000000,0.000000,1.000000" 0.2))
  (bind _returnvalue_9 (Math|Boolean|ORBoolean _returnvalue_8 _returnvalue_7 _returnvalue_6))
  (bind _paintdecalmaterial (Variables|Default|GetPaintDecalMaterial))
  (if (ScalableFloat|GetValueasBool (Variables|Default|GetHF_EnableEnvironmentDecal))
    (Utilities|IsValid _paintdecalmaterial
      (:Is%20Valid
        (Development|PrintString (Utilities|String|ToString(Boolean) _returnvalue_6))
        (bind _returnvalue_10 (Rendering|Decal|SpawnDecalatLocation _paintdecalmaterial (Math|Vector|SelectVector Size (Math|Vector|MakeVector (.x Size) (.z Size) (.y Size)) _returnvalue_9) SpawnLocation (Math|Rotator|SelectRotator (Math|Rotator|MakeRotfromX HitNormal) Rotation _returnvalue_6) (Variables|Default|GetPaintDecalLifetime)))
        (Transformation|AddLocalRotation _returnvalue_10 (Math|Rotator|MakeRotator (Math|Float|SelectFloat (.yaw (Math|Rotator|MakeRotfromZ (Math|Vector|MakeVector 0.0 0.0 -1.0))) (.roll (select _returnvalue_9 "0, 0, 0" (Math|Rotator|SelectRotator (* (Math|Rotator|SelectRotator "0.000000,0.000000,90.000000" "0.000000,0.000000,-90.000000" (Utilities|Operators|Equal(==) _returnvalue "0.000000,1.000000,0.000000" 0.2)) (Math|Float|SelectFloat -1.0 1.0 (< (Math|Vector|DotProduct WeaponDirection "1.000000,0.000000,0.000000") 0))) (* (Math|Rotator|SelectRotator "0.000000,0.000000,-90.000000" "0.000000,0.000000,90.000000" (Utilities|Operators|Equal(==) _returnvalue "1.000000,0.000000,0.000000" 0.2)) (Math|Float|SelectFloat -1.0 1.0 (< (Math|Vector|DotProduct WeaponDirection "0.000000,1.000000,0.000000") 0))) (or (Utilities|Operators|Equal(==) _returnvalue "0.000000,1.000000,0.000000" 0.2) (Utilities|Operators|Equal(==) _returnvalue "0.000000,-1.000000,0.000000" 0.2))))) _returnvalue_6)) false false)
        (Rendering|Components|Decal|SetFadeIn _returnvalue_10 0.0 (Variables|Default|GetPaintDecalFadeInDuration))
        (Rendering|Components|Decal|SetFadeOut _returnvalue_10 (Variables|Default|GetPaintDecalDelayBeforeFade) (Variables|Default|GetPaintDecalFadeOutDuration) true)
        (return _returnvalue_10)))))

(fn ConstructionScript
  (|CallParentFunction))
