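# Sound cue for the IronLog shotgun's screen splat effect, in the ranged weapon projectile effects sound class.
# A Random node picks one of six equally weighted splat waves and passes it through a Modulator for variation, with a boosted cue volume.
# A concurrency override allows two instances per owner and blocks new plays beyond that instead of cutting off playing ones.
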SoundCue /WildEstateWeapons/Gameplay/IronLog/Sound/SC_IronLog_ScreenEffect
root=Modulator_0
nodes [
 {id=Modulator_0, children [Random_0]}
 {id=Random_0, children [WavePlayer_0,WavePlayer_1,WavePlayer_2,WavePlayer_3,WavePlayer_4,WavePlayer_5], properties {weights [1,1,1,1,1,1]}}
 {id=WavePlayer_0, properties {soundWaveAssetPtr=/WildEstateWeapons/Gameplay/IronLog/Sound/Waves/IronLog_Shotgun_Screen_Splat_01}}
 {id=WavePlayer_1, properties {soundWaveAssetPtr=/WildEstateWeapons/Gameplay/IronLog/Sound/Waves/IronLog_Shotgun_Screen_Splat_02}}
 {id=WavePlayer_2, properties {soundWaveAssetPtr=/WildEstateWeapons/Gameplay/IronLog/Sound/Waves/IronLog_Shotgun_Screen_Splat_03}}
 {id=WavePlayer_3, properties {soundWaveAssetPtr=/WildEstateWeapons/Gameplay/IronLog/Sound/Waves/IronLog_Shotgun_Screen_Splat_04}}
 {id=WavePlayer_4, properties {soundWaveAssetPtr=/WildEstateWeapons/Gameplay/IronLog/Sound/Waves/IronLog_Shotgun_Screen_Splat_05}}
 {id=WavePlayer_5, properties {soundWaveAssetPtr=/WildEstateWeapons/Gameplay/IronLog/Sound/Waves/IronLog_Shotgun_Screen_Splat_06}}
]
properties {
 bOverrideConcurrency=true
 maxDistance=2097152
 soundClassObject=&Fort_WeaponRanged_Projectile_Effects
 volumeMultiplier=2.25
 concurrencyOverrides {bLimitToOwner=true, maxCount=2, resolutionRule=PreventNew}
}
