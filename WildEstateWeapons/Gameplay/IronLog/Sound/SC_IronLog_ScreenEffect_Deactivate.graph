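# Sound cue for the deactivation of the IronLog weapon's screen effect, playing a shotgun screen smear sound.
# A fixed delay feeds a Modulator at default settings, which plays one of three equally weighted smear waves through a Random node.
# A concurrency override allows one instance per owner and stops the oldest, with output routed to the ranged weapon projectile effects sound class.
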SoundCue /WildEstateWeapons/Gameplay/IronLog/Sound/SC_IronLog_ScreenEffect_Deactivate
root=Delay_0
nodes [
 {id=Delay_0, children [Modulator_1], properties {delayMax=0.8, delayMin=0.8}}
 {id=Modulator_1, children [Random_1]}
 {id=Random_1, children [WavePlayer_6,WavePlayer_7,WavePlayer_8], properties {weights [1,1,1]}}
 {id=WavePlayer_6, properties {soundWaveAssetPtr=/WildEstateWeapons/Gameplay/IronLog/Sound/Waves/IronLog_Shotgun_Screen_Smear_03}}
 {id=WavePlayer_7, properties {soundWaveAssetPtr=/WildEstateWeapons/Gameplay/IronLog/Sound/Waves/IronLog_Shotgun_Screen_Smear_02}}
 {id=WavePlayer_8, properties {soundWaveAssetPtr=/WildEstateWeapons/Gameplay/IronLog/Sound/Waves/IronLog_Shotgun_Screen_Smear_01}}
]
properties {
 bOverrideConcurrency=true
 maxDistance=2097152
 soundClassObject=&Fort_WeaponRanged_Projectile_Effects
 volumeMultiplier=1
 concurrencyOverrides {bLimitToOwner=true, maxCount=1, resolutionRule=StopOldest}
}
